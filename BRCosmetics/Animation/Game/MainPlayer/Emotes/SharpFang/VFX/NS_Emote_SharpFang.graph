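# Niagara VFX for the SharpFang main player emote: a liquid splash burst plus a trail left by movement.
# Burst_Droplets and Burst_Splash spawn instant sprite bursts with drag and gravity, while Droplet_Trail, Splash_Trail and Snail_Ribbon_Trail spawn by distance traveled, the last as a wide ribbon.
# All five emitters take color from System.LiquidColor and animate materials through dynamic parameters; splash sprites also use SubUV.
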NiagaraSystem /BRCosmetics/Animation/Game/MainPlayer/Emotes/SharpFang/VFX/NS_Emote_SharpFang
dependencies {
 usedDataInterfaces [&NiagaraDataInterfaceCurve, &NiagaraDataInterfaceVector2DCurve]
 usedDynamicInputs [&Vector2DFromCurve]
 usedModules [
  &self:SystemSpawnScript.SystemScriptSource.SystemScriptGraph.NiagaraNodeAssignment_0.SetVariables_#1
  &SystemState
 ]
}
emitters {
 Burst_Droplets {
  emitterUpdateScript [
   &EmitterState(Life Cycle Mode=::NewEnumerator1, Loop Behavior=::NewEnumerator1, Loop Duration=1, UseLoopDelay=false)
   &SpawnBurst_Instantaneous(Spawn Count=60, Use Spawn Probability=false, UseLoopCountLimit=false)
  ]
  particleSpawnScript [
   &InitializeParticle(
    Color=$System.LiquidColor
    Lifetime Max=1
    Lifetime Min=0.25
    Lifetime Mode=::NewEnumerator1
    Position Offset=<0 0 30>
    Sprite Size Max=<4 12>
    Sprite Size Min=<2 6>
    Sprite Size Mode=::NewEnumerator2
    UsePositionOffset=true
    Write Material Random=true
   )
   &ShapeLocation(
    Hemisphere Distribution=<1 0.66>
    MaskSpawnGroup=false
    Offset Mode=::NewEnumerator0
    Override Seed=false
    Sphere Radius=25
    Sphere Surface Distribution=0.8
    UseOwnerScale=true
   )
   &AddVelocity(
    Module.Use Falloff Distance=false
    Origin Offset=<0 0 20>
    Override Seed=false
    Velocity Mode=::NewEnumerator1
    Velocity Speed=&RandomRangeFloat(Minimum=200, Maximum=500)
   )
   &DynamicMaterialParameters(
    Index 0 Param 1=&UniformRangedFloat(Minimum=10, Maximum=10)
    Module.Param0WriteEnabled=true
    Module.Param1WriteEnabled=false
    Module.Param2WriteEnabled=false
    Module.Param3WriteEnabled=false
   )
  ]
  particleUpdateScript [
   &ParticleState()
   &Drag(Drag=&RandomRangeFloat(Minimum=0.25, Maximum=0.5), Use Linear Drag=true, Use Rotational Drag=false)
   &GravityForce(Gravity=&MaskVectorBySpawnGroup(Masked Vector=<0 0 -980>, Passthrough Vector=&RandomRangeVector(Minimum=<0 0 -25>, Maximum=<0 0 -100>)))
   &SolveForcesAndVelocity(Clamp Velocity=false, Limit Acceleration=false)
   &ScaleColor(
    Scale Alpha=&FloatFromCurve(
     FloatCurve=DataInterface() {propertyValues {Curve {keys [
         {arriveTangent=28.4939, interpMode=RCIM_Cubic, leaveTangent=28.4939, tangentMode=RCTM_User, time=0, value=0}
         {arriveTangent=0.002, interpMode=RCIM_Cubic, leaveTangent=0.002, time=0.1183, value=1.0113}
         {arriveTangent=-3.6864, interpMode=RCIM_Cubic, leaveTangent=-3.6864, tangentMode=RCTM_User, time=1.0011, value=0.002}
        ]}}}
     Scale Curve=0.7
    )
    Scale RGB=&VectorFromFloat(Value=&RandomRangeFloat(Minimum=0.625, Maximum=1))
    ScaleA=true
    ScaleRGB=true
   )
  ]
  rendererClasses [&NiagaraSpriteRendererProperties]
  renderers [{rendererClass=&NiagaraSpriteRendererProperties}]
 }
 Burst_Splash {
  emitterUpdateScript [
   &EmitterState(Life Cycle Mode=::NewEnumerator1, Loop Behavior=::NewEnumerator1, Loop Duration=1, UseLoopDelay=false)
   &SpawnBurst_Instantaneous(Spawn Count=15, Use Spawn Probability=false, UseLoopCountLimit=false) {module=SpawnBurst_Instantaneous001}
  ]
  particleSpawnScript [
   &InitializeParticle(
    Color=$System.LiquidColor
    Lifetime Max=0.5
    Lifetime Min=0.4
    Lifetime Mode=::NewEnumerator1
    Position Offset=<0 0 30>
    Sprite Rotation Mode=::NewEnumerator2
    Sprite Size Mode=::NewEnumerator1
    Uniform Sprite Size Max=120
    Uniform Sprite Size Min=110
    UsePositionOffset=true
    Write Material Random=true
   ) {module=InitializeParticle001}
   &ShapeLocation(
    Hemisphere Distribution=<1 0.66>
    MaskSpawnGroup=false
    Offset Mode=::NewEnumerator0
    Override Seed=false
    Sphere Radius=25
    Sphere Surface Distribution=0.8
    UseOwnerScale=true
   )
   &AddVelocity(
    Module.Use Falloff Distance=false
    Origin Offset=<0 0 30>
    Override Seed=false
    Velocity Mode=::NewEnumerator1
    Velocity Speed=&RandomRangeFloat(Minimum=200, Maximum=250)
   ) {module=AddVelocity001}
   /Niagara/Modules/Update/SubUV/SubUVAnimation(SubUV Animation Mode=::NewEnumerator1)
  ]
  particleUpdateScript [
   &ParticleState()
   &InheritVelocity(
    Module.Limit Speed=true
    Velocity Limit=10000
    Velocity Scale=&Multiply_VectorByFloat(Vector=<-0.03 -0.03 -0.03>, Float=0.5)
    Velocity Source=&TransformVector(Vector=$Engine.Owner.Velocity, Source Space=::World, Destination Space=::Local)
   )
   &SpriteRotationRate(Rotation Rate=&RandomRangeFloat(Minimum=-90, Maximum=90))
   &Drag(Drag=0.5, Use Linear Drag=true, Use Rotational Drag=false)
   &GravityForce(Gravity=<0 0 -900>)
   &SolveForcesAndVelocity(Clamp Velocity=false, Limit Acceleration=false)
   &ScaleSpriteSize(Scale Factor=&Vector2DFromFloat(Value=&FloatFromCurve(
      FloatCurve=DataInterface() {propertyValues {Curve {keys [
          {arriveTangent=1.6115, interpMode=RCIM_Cubic, leaveTangent=1.6115, tangentMode=RCTM_User, time=0, value=0.33}
          {interpMode=RCIM_Cubic, tangentMode=RCTM_User, time=1, value=1}
         ]}}}
      Scale Curve=0.65
     )))
   &ScaleColor(
    Scale RGB=&VectorFromFloat(Value=&RandomRangeFloat(Minimum=0.75, Maximum=1.5))
    ScaleA=true
    ScaleRGB=true
   )
   &DynamicMaterialParameters(
    Index 0 Param 1=&RandomRangeFloat(Minimum=0, Maximum=1)
    Index 0 Param 3=&FloatFromCurve(FloatCurve=DataInterface() {propertyValues {Curve {keys [{interpMode=RCIM_Cubic, tangentMode=RCTM_User, time=0, value=0}, {interpMode=RCIM_Cubic, tangentMode=RCTM_User, time=1, value=1}]}}})
    Index 0 Param 4=&FloatFromCurve(FloatCurve=DataInterface() {propertyValues {Curve {keys [<at 0 0 RCIM_Cubic>, {interpMode=RCIM_Cubic, leaveTangent=1.5276, time=1, value=1}]}}})
    Module.Param0WriteEnabled=true
    Module.Param1WriteEnabled=false
    Module.Param2WriteEnabled=true
    Module.Param3WriteEnabled=true
   )
  ]
  rendererClasses [&NiagaraSpriteRendererProperties]
  renderers [{rendererClass=&NiagaraSpriteRendererProperties}]
 }
 Droplet_Trail {
  emitterUpdateScript [
   &EmitterState()
   &SpawnPerUnit(
    Module.Use Max Movement Threshold=true
    Module.Use Movement Tolerance=false
    Spawn Spacing:NiagaraFloat=&MakeFloatFromInt(Int=10)
    Use Spawn Probability=false
   )
   &SpawnRate(SpawnRate=60, Use Spawn Probability=false) {enabled=false}
  ]
  particleSpawnScript [
   &InitializeParticle(
    Color=$System.LiquidColor
    Lifetime Max=0.75
    Lifetime Min=0.25
    Lifetime Mode=::NewEnumerator1
    Position Offset=<0 0 30>
    Sprite Size Max=<4 12>
    Sprite Size Min=<2 6>
    Sprite Size Mode=::NewEnumerator2
    UsePositionOffset=true
    Write Material Random=true
   )
   &DynamicMaterialParameters(
    Index 0 Param 1=&UniformRangedFloat(Minimum=10, Maximum=10)
    Module.Param0WriteEnabled=true
    Module.Param1WriteEnabled=false
    Module.Param2WriteEnabled=false
    Module.Param3WriteEnabled=false
   )
   &ShapeLocation(
    Additional Quaternion Rotation=true
    Additional Yaw / Pitch / Roll=<0 0 0.5>
    Hemisphere Distribution=<0.5 0.5>
    MaskSpawnGroup=false
    Offset Mode=::NewEnumerator0
    Override Seed=false
    Rotation Mode=::NewEnumerator3
    Sphere Radius=25
    Sphere Surface Distribution=1
    UseOwnerScale=true
   )
   &AddVelocity(
    Module.Use Falloff Distance=false
    Origin Offset=<0 5 -40>
    Override Seed=false
    Velocity Mode=::NewEnumerator1
    Velocity Speed=&RandomRangeFloat(Minimum=250, Maximum=500)
   )
  ]
  particleUpdateScript [
   &ParticleState()
   &Drag(Use Linear Drag=true, Use Rotational Drag=false)
   &GravityForce(Gravity=&MaskVectorBySpawnGroup(Masked Vector=<0 0 -980>, Passthrough Vector=&RandomRangeVector(Minimum=<0 0 -25>, Maximum=<0 0 -100>)))
   &SolveForcesAndVelocity(Clamp Velocity=false, Limit Acceleration=false)
   &ScaleColor(
    Scale Alpha=&FloatFromCurve(FloatCurve=DataInterface() {propertyValues {Curve {keys [
         {arriveTangent=28.4939, interpMode=RCIM_Cubic, leaveTangent=28.4939, tangentMode=RCTM_User, time=0, value=0}
         {arriveTangent=0.002, interpMode=RCIM_Cubic, leaveTangent=0.002, time=0.1183, value=1.0113}
         {arriveTangent=-3.6864, interpMode=RCIM_Cubic, leaveTangent=-3.6864, tangentMode=RCTM_User, time=1.0011, value=0.002}
        ]}}})
    Scale RGB=&VectorFromFloat(Value=&RandomRangeFloat(Minimum=0.625, Maximum=1))
    ScaleA=true
    ScaleRGB=true
   )
  ]
  rendererClasses [&NiagaraSpriteRendererProperties]
  renderers [{rendererClass=&NiagaraSpriteRendererProperties}]
 }
 Snail_Ribbon_Trail {
  emitterUpdateScript [
   &EmitterState()
   &SpawnPerUnit(
    Module.Use Max Movement Threshold=false
    Module.Use Movement Tolerance=false
    Spawn Spacing:NiagaraFloat=&MakeFloatFromInt(Int=25)
    Use Spawn Probability=false
   )
  ]
  particleSpawnScript [&InitializeParticle(
    Color=$System.LiquidColor
    Lifetime=1.5
    Ribbon Facing Mode=::NewEnumerator1
    Ribbon Width=200
    Ribbon Width Mode=::NewEnumerator1
    UsePositionOffset=false
    Write Material Random=true
   )]
  particleUpdateScript [
   &ParticleState()
   &DynamicMaterialParameters(
    Index 0 Param 2=&FloatFromCurve(FloatCurve=DataInterface() {propertyValues {Curve {keys [
         <at 0 0 RCIM_Cubic>
         {interpMode=RCIM_Cubic, leaveTangent=-2.6862, tangentMode=RCTM_Break, time=0.1, value=1}
         {arriveTangent=-0.5259, interpMode=RCIM_Cubic, leaveTangent=-0.5259, tangentMode=RCTM_User, time=1, value=0}
        ]}}})
    Index 0 Param 3=&FloatFromCurve(FloatCurve=DataInterface() {propertyValues {Curve {keys [{tangentMode=RCTM_User, time=0, value=0}, {tangentMode=RCTM_User, time=1, value=1.5}]}}})
    Index 0 Param 4=&FloatFromCurve(
     FloatCurve=DataInterface() {propertyValues {Curve {keys [{tangentMode=RCTM_User, time=0, value=1}, {tangentMode=RCTM_User, time=1, value=1.15}]}}}
     Scale Curve=3
    )
    Module.Param0WriteEnabled=false
    Module.Param1WriteEnabled=true
    Module.Param2WriteEnabled=false
    Module.Param3WriteEnabled=false
   )
   &ScaleColor(
    Scale Alpha=&FloatFromCurve(
     FloatCurve=DataInterface() {propertyValues {Curve {keys [{interpMode=RCIM_Cubic, tangentMode=RCTM_User, time=0.9, value=1}, {interpMode=RCIM_Cubic, leaveTangent=-7.011, time=1, value=0}]}}}
     Scale Curve=0.25
    )
    Scale RGB=&VectorFromFloat(Value=&RandomRangeFloat(Minimum=2.5, Maximum=2.5))
    ScaleA=true
    ScaleRGB=true
   )
  ]
  rendererClasses [&NiagaraRibbonRendererProperties]
  renderers [{rendererClass=&NiagaraRibbonRendererProperties}]
 }
 Splash_Trail {
  emitterUpdateScript [
   &EmitterState()
   &SpawnPerUnit(
    Module.Use Max Movement Threshold=true
    Module.Use Movement Tolerance=false
    Spawn Spacing:NiagaraFloat=&MakeFloatFromInt(Int=10)
    Use Spawn Probability=false
   )
  ]
  particleSpawnScript [
   &InitializeParticle(
    Color=$System.LiquidColor
    Lifetime Max=0.5
    Lifetime Min=0.33
    Lifetime Mode=::NewEnumerator1
    Position Offset=<0 0 30>
    Sprite Rotation Mode=::NewEnumerator2
    Sprite Size Mode=::NewEnumerator1
    Uniform Sprite Size Max=120
    Uniform Sprite Size Min=110
    UsePositionOffset=true
    Write Material Random=true
   )
   &ShapeLocation(
    Additional Quaternion Rotation=true
    Additional Yaw / Pitch / Roll=<0 0 0.5>
    Hemisphere Distribution=<0.5 0.5>
    MaskSpawnGroup=false
    Offset Mode=::NewEnumerator0
    Override Seed=false
    Rotation Mode=::NewEnumerator3
    Sphere Radius=25
    Sphere Surface Distribution=1
    UseOwnerScale=true
   )
   &AddVelocity(
    Module.Use Falloff Distance=false
    Origin Offset=<0 10 -40>
    Override Seed=false
    Velocity Mode=::NewEnumerator1
    Velocity Speed=&RandomRangeFloat(Minimum=100, Maximum=200)
   )
   &SubUVAnimation(SubUV Animation Mode=::NewEnumerator1)
  ]
  particleUpdateScript [
   &ParticleState()
   &Drag(Drag=0.25, Use Linear Drag=true, Use Rotational Drag=false)
   &SpriteRotationRate(Rotation Rate=&RandomRangeFloat(Minimum=-90, Maximum=90))
   &GravityForce(Gravity=&MaskVectorBySpawnGroup(Masked Vector=<0 0 -980>, Passthrough Vector=&RandomRangeVector(Minimum=<0 0 -25>, Maximum=<0 0 -100>)))
   &SolveForcesAndVelocity(Clamp Velocity=false, Limit Acceleration=false)
   &ScaleSpriteSize(Scale Factor=&Vector2DFromFloat(Value=&FloatFromCurve(
      FloatCurve=DataInterface() {propertyValues {Curve {keys [
          {arriveTangent=1.6115, interpMode=RCIM_Cubic, leaveTangent=1.6115, tangentMode=RCTM_User, time=0, value=0.33}
          {interpMode=RCIM_Cubic, tangentMode=RCTM_User, time=1, value=1}
         ]}}}
      Scale Curve=0.65
     )))
   &ScaleColor(
    Scale RGB=&VectorFromFloat(Value=&RandomRangeFloat(Minimum=0.75, Maximum=1.5))
    ScaleA=true
    ScaleRGB=true
   )
   &DynamicMaterialParameters(
    Index 0 Param 1=&RandomRangeFloat(Minimum=0, Maximum=1)
    Index 0 Param 3=&FloatFromCurve(FloatCurve=DataInterface() {propertyValues {Curve {keys [{interpMode=RCIM_Cubic, tangentMode=RCTM_User, time=0, value=0}, {interpMode=RCIM_Cubic, tangentMode=RCTM_User, time=1, value=1}]}}})
    Index 0 Param 4=&FloatFromCurve(FloatCurve=DataInterface() {propertyValues {Curve {keys [<at 0 0 RCIM_Cubic>, {interpMode=RCIM_Cubic, leaveTangent=1.5276, time=1, value=1}]}}})
    Module.Param0WriteEnabled=true
    Module.Param1WriteEnabled=false
    Module.Param2WriteEnabled=true
    Module.Param3WriteEnabled=true
   ) {module=DynamicMaterialParameters001}
  ]
  rendererClasses [&NiagaraSpriteRendererProperties]
  renderers [{rendererClass=&NiagaraSpriteRendererProperties}]
 }
}
summary {
 systemName=NS_Emote_SharpFang
 emitters [
  {emitterName=Snail_Ribbon_Trail, rendererClasses [&NiagaraRibbonRendererProperties]}
  {emitterName=Burst_Droplets, rendererClasses [&NiagaraSpriteRendererProperties]}
  {emitterName=Burst_Splash, rendererClasses [&NiagaraSpriteRendererProperties]}
  {emitterName=Droplet_Trail, rendererClasses [&NiagaraSpriteRendererProperties]}
  {emitterName=Splash_Trail, rendererClasses [&NiagaraSpriteRendererProperties]}
 ]
 userVariables []
}
